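# Niagara effect for the VigorKnackGolf back bling that shows a kill count and reacts with a pulse and shockwave.
# Base layers sprites and a mesh for the one- or two-digit previous count, switching frames and tint across three kill-count stages.
# ShockwaveBurst adds stage-colored mesh waves and a sprite flash, partly keyed to the current count.
# Particles persist past their lifetime, with curves indexed by system reactivity intensity driving the animation instead of particle age.
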NiagaraSystem /BRCosmetics/Accessories/FORT_Backpacks/Backpack_VigorKnackGolf/FX/NS_Backpack_VigorKnackGolf_IdleReactivity
dependencies {
 usedDataInterfaces [
  &NiagaraDataInterfaceCurve
  &NiagaraDataInterfaceSpriteRendererInfo
  &NiagaraDataInterfaceVector2DCurve
 ]
 usedDynamicInputs [
  &MakeIntFromFloat
  &ClampFloat
  &Subtract_Float
  &SetBoolByFloatComparison
  /Niagara/DynamicInputs/Subtract/Subtract_Integer
  &FN_FrontEndFloatModifier
 ]
 usedModules [
  &SystemState
  &self:SystemSpawnScript.SystemScriptSource.SystemScriptGraph.NiagaraNodeAssignment_2.SetVariables_#1
  &self:SystemSpawnScript.SystemScriptSource.SystemScriptGraph.NiagaraNodeAssignment_1.SetVariables_#2
  &self:SystemSpawnScript.SystemScriptSource.SystemScriptGraph.NiagaraNodeAssignment_4.SetVariables_#3
  &self:SystemSpawnScript.SystemScriptSource.SystemScriptGraph.NiagaraNodeAssignment_0.SetVariables_#4
  &self:SystemSpawnScript.SystemScriptSource.SystemScriptGraph.NiagaraNodeAssignment_3.SetVariables_#5
  &self:SystemSpawnScript.SystemScriptSource.SystemScriptGraph.NiagaraNodeAssignment_5.SetVariables_#6
 ]
}
emitters {
 Base {
  emitterUpdateScript [
   &EmitterState(
    Inactive Response=::NewEnumerator1
    Life Cycle Mode=::NewEnumerator1
    Loop Behavior=::NewEnumerator1
    Loop Duration Mode=::NewEnumerator1
    UseLoopDelay=false
   )
   /SetVariables(Emitter.NumberSpacing:NiagaraFloat=&MakeCustomFloatFromBool(Bool=$System.IsStage03, True Float=3, False Float=2.35)) {isSetParameters=true}
   /SetVariables(Emitter.IsSingleDigit:NiagaraBool=&SetBoolByIntComparison(A=$System.SystemKillCountPrevious, B=10, Comparison Type=::NewEnumerator6)) {isSetParameters=true}
   /SetVariables(Emitter.IsDoubleDigit:NiagaraBool=&Boolean_NotOperation(Bool=$Emitter.IsSingleDigit)) {isSetParameters=true}
   /SetVariables(Emitter.ReactivityAge:NiagaraFloat=&Divide_Float(A=$System.ReactivityIntensity, B=0.35)) {isSetParameters=true}
   &SpawnBurst_Instantaneous(Use Spawn Probability=false, UseLoopCountLimit=false)
   &SpawnBurst_Instantaneous(Spawn Group=1, Use Spawn Probability=false, UseLoopCountLimit=false) {module=SpawnBurst_Instantaneous002}
   &SpawnBurst_Instantaneous(Spawn Group=2, Use Spawn Probability=false, UseLoopCountLimit=false) {module=SpawnBurst_Instantaneous001}
   &SpawnBurst_Instantaneous(Spawn Group=3, Use Spawn Probability=false, UseLoopCountLimit=false) {module=SpawnBurst_Instantaneous003}
  ]
  particleSpawnScript [
   /SetVariables(Particles.VisibilityTag=$Emitter.SpawnGroup) {isSetParameters=true}
   &InitializeParticle(
    Color Mode=::NewEnumerator0
    Lifetime=5
    Position Offset=&AddVectorToPosition(
     Position=<0 -3.75 0>
     Vector=&MergeHelper_VectorSwitch(0=<0 0 0>, 1=<0 -0.75 0>, 2=<0 -1.5 0>, 3=<0 -2 0>, Index=$Particles.VisibilityTag, Num=::NewEnumerator2)
    )
    Position Offset Coordinate Space=::Local
    Sprite Size Mode=::NewEnumerator3
    Uniform Sprite Size=30
    UsePositionOffset=true
    Write Material Random=true
   )
   &SpriteFacingAndAlignment(Sprite Alignment=<0 0 1>, Sprite Facing=<0 -1 0>, WriteAlignment=true, WriteFacing=true)
  ]
  particleUpdateScript [
   &ParticleState(Kill Particles When Lifetime Has Elapsed=false)
   &ScaleColor(
    Scale RGB=&Multiply_Vector(
     A=&MakeVectorFromLinearColorRGB(Color=&MakeCustomLinearColorFromBool(Bool=$System.IsStage03, True Color=<rgba 0.7565 0.5689 0.4917 1>, False Color=<rgba 0.5 0.5 0.5 1>))
     B=&VectorFromFloat(Value=&FNTimeOfDayScalar(Night Scale=0.8, Morning Scale=0.85, Evening Scale=0.9))
    )
    ScaleA=true
    ScaleRGB=true
   )
   &ScaleSpriteSize(Scale Sprite Size Mode=::NewEnumerator0, Uniform Scale Factor=$System.SystemScale)
   &SubUVAnimation(
    Direct Frame Index:NiagaraFloat=&MakeFloatFromInt(Int=&MakeCustomIntFromBool(
      Bool=$System.IsStage01
      True INT=&MergeHelper_IntSwitch(0=0, 1=1, 2=2, 3=?, 4=?, 5=?, Index=$Particles.VisibilityTag, Num=::NewEnumerator1)
      False INT=&MakeCustomIntFromBool(
       Bool=$System.IsStage02
       True INT=&MergeHelper_IntSwitch(0=3, 1=4, 2=5, 3=?, 4=?, 5=?, Index=$Particles.VisibilityTag, Num=::NewEnumerator1)
       False INT=&MergeHelper_IntSwitch(0=6, 1=7, 2=8, 3=?, 4=?, 5=?, Index=$Particles.VisibilityTag, Num=::NewEnumerator1)
      )
     ))
    Output Alternate Renderer Binding=true
    "Sprite Renderer "=DataInterface() {propertyValues {SpriteRenderer=&self:Base_0.NiagaraSpriteRendererProperties_0}}
    SubUV Animation Mode:ENiagara_SubUVLookupModeV2=::NewEnumerator5
   )
   /SetVariables(
    Particles.Digit1:NiagaraFloat=&MakeFloatFromInt(Int=/Niagara/DynamicInputs/Divide/Divide_Int(A:NiagaraInt32=$System.SystemKillCountPrevious, B:NiagaraInt32=10))
    Particles.Digit2:NiagaraFloat=&MakeFloatFromInt(Int=$System.SystemKillCountPrevious)
   ) {isSetParameters=true}
   /SetVariables(Particles.GlowColor=&MergeHelper_ColorSwitch(
     0=<rgba 0.0527 0.5 0.0081 1>
     1=&MergeHelper_ColorSwitch(
      0=<rgba 0.2333 0.4 0 1>
      1=<rgba 0.55 0.2108 0 1>
      2=?
      3=?
      Index=&MakeCustomIntFromBool(Bool=$System.IsStage02, True INT=0, False INT=1)
      Num=::NewEnumerator0
     )
     2=?
     3=?
     Index=&MakeCustomIntFromBool(Bool=$System.IsStage01, True INT=0, False INT=1)
     Num=::NewEnumerator0
    )) {isSetParameters=true}
   /SetVariables(
    Particles.OffsetLocation01:NiagaraPosition=&AddVectorToPosition(
     Position=$Particles.Initial.Position
     Vector=&MakeCustomVectorFromBool(
      Bool=$Emitter.IsSingleDigit
      True Vector=<0 0 0>
      False Vector=&MakeVector(X=&Multiply_Float(A=$Emitter.NumberSpacing, B=-1), Y=0, Z=0)
     )
    )
    Particles.OffsetLocation02:NiagaraPosition=&AddVectorToPosition(
     Position=$Particles.Initial.Position
     Vector=&MakeCustomVectorFromBool(
      Bool=$Emitter.IsSingleDigit
      True Vector=<0 0 0>
      False Vector=&MakeVector(X=$Emitter.NumberSpacing, Y=0, Z=0)
     )
    )
   ) {isSetParameters=true}
   &FN_Sprite_Attributes(
    Camera Offset=1
    Create Camera Offset Attribute=true
    Create Color Attribute=true
    Create SpriteSize Attribute=true
    Module Output:ENiagara_AttributeSamplingApplyOutput=::NewEnumerator1
    SpriteSize=&Multiply_Vector2D(
     A=&Vector2DFromCurve(
      Vector2Curve=DataInterface() {propertyValues {
        XCurve {keys [
          <at 0.85 1 RCIM_Cubic>
          {arriveTangent=-24.0271, interpMode=RCIM_Cubic, leaveTangent=-24.0271, tangentMode=RCTM_User, time=0.9, value=2}
          <at 1 1>
         ]}
        YCurve {keys [
          {arriveTangent=2.1818, interpMode=RCIM_Cubic, leaveTangent=2.1818, tangentMode=RCTM_User, time=0.245, value=1}
          {arriveTangent=0.3892, interpMode=RCIM_Cubic, leaveTangent=0.3892, tangentMode=RCTM_User, time=0.4, value=1.2}
          {interpMode=RCIM_Cubic, tangentMode=RCTM_User, time=0.8, value=1.25}
          {arriveTangent=-1.25, interpMode=RCIM_Cubic, leaveTangent=-1.25, time=0.9, value=0.845}
          <at 1 1 RCIM_Cubic>
         ]}
       }}
      CurveIndex=$Emitter.ReactivityAge
      ScaleCurve=&Vector2DFromFloat(Value=$System.SystemScale)
     )
     B=&Vector2DFromFloat(Value=&MakeCustomFloatFromBool(Bool=$Emitter.IsSingleDigit, True Float=10.5, False Float=8.35))
    )
   )
   &FN_Sprite_Attributes(
    Color=$Particles.GlowColor
    Create Color Attribute=true
    Create SpriteSize Attribute=true
    Module Output:ENiagara_AttributeSamplingApplyOutput=::NewEnumerator1
    SpriteSize=&Multiply_Vector2D(
     A=&Multiply_Vector2D_ByFloat(Float=$System.SystemScale, Vector2D=$Particles.FN_Sprite_Attributes.SpriteSize)
     B=<1 1.05>
    )
   ) {module=FN_Sprite_Attributes001}
   &FN_Sprite_Attributes(
    Color=&MakeCustomLinearColorFromBool(
     Bool=$System.IsStage01
     True Color=<rgba 0.078 0.233 0.066 10>
     False Color=&MakeCustomLinearColorFromBool(Bool=$System.IsStage02, True Color=<rgba 0.1375 0.2348 0.1121 10>, False Color=<rgba 0.348 0.1574 0.0757 10>)
    )
    Create Color Attribute=true
    Create SpriteSize Attribute=true
    Module Output:ENiagara_AttributeSamplingApplyOutput=::NewEnumerator1
    SpriteSize=&Multiply_Vector2D_ByFloat(
     Float=$System.SystemScale
     Vector2D=&MakeCustomVector2DFromBool(Bool=$System.IsStage03, True Vector2D=<10 14>, False Vector2D=<10 14.4>)
    )
   ) {module=FN_Sprite_Attributes002}
   &FN_Mesh_Attributes(
    Camera Offset=1
    Color=/Niagara/DynamicInputs/LinearColor/MakeLinearColorFromVector4(Vector:Vector4f=&Multiply_Vector4(
      A=&MakeVector4_FromLinearColor(LinearColor=&MakeCustomLinearColorFromBool(
        Bool=$System.IsStage01
        True Color=<rgba 1.2376 3 1.438 0.1>
        False Color=&MakeCustomLinearColorFromBool(Bool=$System.IsStage02, True Color=<rgba 1 0.9647 0.1441 0.2>, False Color=<rgba 1 0.3372 0.0343 0.3>)
       ))
      B=&MakeVector4FromVectorAndFloat(
       "Vector (RGB)"=<1 1 1>
       "Float (Alpha)"=&MakeCustomFloatFromBool(Bool=$System.IsStage03, True Float=0.7, False Float=0.3)
      )
     ))
    Create Camera Offset Attribute=true
    Create Color Attribute=true
    Create Scale Attribute=true
    Module Output=::NewEnumerator1
    Scale=&Multiply_VectorByFloat(
     Vector=&Multiply_Vector(
      A=<1 1 1.5>
      B=&MakeCustomVectorFromBool(
       Bool=$System.IsStage03
       True Vector=&VectorFromFloat(Value=0.9)
       False Vector=&VectorFromFloat(Value=1)
      )
     )
     Float=$System.SystemScale
    )
   )
  ]
  rendererClasses [&NiagaraSpriteRendererProperties, &NiagaraMeshRendererProperties]
  renderers [
   {rendererClass=&NiagaraSpriteRendererProperties}
   {rendererClass=&NiagaraSpriteRendererProperties, rendererIndex=1}
   {rendererClass=&NiagaraSpriteRendererProperties, rendererIndex=2}
   {rendererClass=&NiagaraSpriteRendererProperties, rendererIndex=3}
   {rendererClass=&NiagaraSpriteRendererProperties, rendererIndex=4}
   {rendererClass=&NiagaraSpriteRendererProperties, rendererIndex=5}
   {rendererClass=&NiagaraSpriteRendererProperties, rendererIndex=6}
   {rendererClass=&NiagaraSpriteRendererProperties, rendererIndex=7}
   {rendererClass=&NiagaraMeshRendererProperties, rendererIndex=8}
  ]
 }
 ShockwaveBurst {
  emitterUpdateScript [
   &EmitterState(
    Inactive Response=::NewEnumerator1
    Life Cycle Mode=::NewEnumerator1
    Loop Behavior=::NewEnumerator1
    Loop Duration Mode=::NewEnumerator1
    UseLoopDelay=false
   )
   &SpawnBurst_Instantaneous(Use Spawn Probability=false, UseLoopCountLimit=false)
   /SetVariables(Emitter.ReactivityAge:NiagaraFloat=&Divide_Float(A=$System.ReactivityIntensity, B=0.5)) {isSetParameters=true}
   /SetVariables(
    Emitter.IsStage01:NiagaraBool=&Boolean_AndOperation(
     Bool1=&SetBoolByIntComparison(A=$System.SystemKillCount, B=0, Comparison Type=::NewEnumerator2)
     Bool2=&SetBoolByIntComparison(A=$System.SystemKillCount, B=4, Comparison Type=::NewEnumerator7)
    )
    Emitter.IsStage02:NiagaraBool=&Boolean_AndOperation(
     Bool1=&SetBoolByIntComparison(A=$System.SystemKillCount, B=5, Comparison Type=::NewEnumerator2)
     Bool2=&SetBoolByIntComparison(A=$System.SystemKillCount, B=9, Comparison Type=::NewEnumerator7)
    )
   ) {isSetParameters=true}
   /SetVariables(Emitter.IsStage03:NiagaraBool=&Boolean_AndOperation(
     Bool1=&Boolean_NotOperation(Bool=$System.IsStage01)
     Bool2=&Boolean_NotOperation(Bool=$System.IsStage02)
    )) {isSetParameters=true}
  ]
  particleSpawnScript [
   /SetVariables(Particles.VisibilityTag=$Emitter.SpawnGroup) {isSetParameters=true}
   &InitializeParticle(
    Color Mode=::NewEnumerator0
    Lifetime=0.35
    Mesh Scale Mode=::NewEnumerator3
    Mesh Uniform Scale=0.296
    Position Offset=<0 -3.75 0>
    Position Offset Coordinate Space=::Local
    Sprite Rotation Mode=::NewEnumerator2
    Sprite UV Mode=::NewEnumerator3
    UsePositionOffset=true
    Write Material Random=true
   )
   &InitialMeshOrientation(Mesh Orientation Mode=::NewEnumerator7, Rotation=<0 0.125 0>, Use Rotation Vector=true)
   &SpriteFacingAndAlignment(
    Alignment Coordinate Space=::Local
    Facing Coordinate Space=::Local
    Sprite Alignment=<1 0 0>
    Sprite Facing=<0 -1 0>
    WriteAlignment=true
    WriteFacing=true
   )
  ]
  particleUpdateScript [
   &ParticleState(Kill Particles When Lifetime Has Elapsed=false)
   &ScaleColor(
    Scale Alpha=&FloatFromCurve(
     FloatCurve=DataInterface() {propertyValues {Curve {keys [
         <at 0 0 RCIM_Constant>
         {arriveTangent=-4.018, interpMode=RCIM_Cubic, leaveTangent=-4.018, tangentMode=RCTM_User, time=0.435, value=1}
         <at 1 0 RCIM_Cubic>
        ]}}}
     CurveIndex=$Emitter.ReactivityAge
     Scale Curve=2.5
    )
    Scale RGB=&Multiply_VectorByFloat(
     Vector=&Multiply_Vector(
      A=&MakeVectorFromLinearColorRGB(Color=&MergeHelper_ColorSwitch(
        0=<rgba 0.0908 1 0 1>
        1=&MergeHelper_ColorSwitch(
         0=<rgba 0.4569 1 0 1>
         1=<rgba 1 0.4038 0 1>
         2=?
         3=?
         Index=&MakeCustomIntFromBool(Bool=$Emitter.IsStage02, True INT=0, False INT=1)
         Num=::NewEnumerator0
        )
        2=?
        3=?
        Index=&MakeCustomIntFromBool(Bool=$Emitter.IsStage01, True INT=0, False INT=1)
        Num=::NewEnumerator0
       ))
      B=&VectorFromFloat(Value=&FNTimeOfDayScalar(Night Scale=0.775, Morning Scale=0.8, Evening Scale=0.85))
     )
     Float=&FloatFromCurve(
      FloatCurve=DataInterface() {propertyValues {Curve {keys [<at 0 5 RCIM_Constant>,<at 0.435 5>,<at 0.635 1>,<at 1 0>]}}}
      CurveIndex=$Emitter.ReactivityAge
     )
    )
    ScaleA=true
    ScaleRGB=true
   )
   &ScaleMeshSize(Scale Factor=&VectorFromFloat(Value=&FloatFromCurve(
      FloatCurve=DataInterface() {propertyValues {Curve {keys [{arriveTangent=2.0091, interpMode=RCIM_Cubic, leaveTangent=2.0091, tangentMode=RCTM_User, time=0.435, value=1}, <at 1 1.5>]}}}
      CurveIndex=$Emitter.ReactivityAge
     )))
   &DynamicMaterialParameters(
    Index 0 Param 1=&FloatFromCurve(FloatCurve=DataInterface() {propertyValues {Curve {keys [<at 0.435 0.36>,<at 1 0>]}}}, CurveIndex=$Emitter.ReactivityAge)
    Index 0 Param 2=&FloatFromCurve(FloatCurve=DataInterface() {propertyValues {Curve {keys [<at 0.435 0>,<at 1 1>]}}}, CurveIndex=$Emitter.ReactivityAge)
    Index 0 Param 3=&FloatFromCurve(FloatCurve=DataInterface() {propertyValues {Curve {keys [<at 0.435 0>,<at 1 0.2>]}}}, CurveIndex=$Emitter.ReactivityAge)
    Module.Index0Float0Write=true
    Module.Index0Float1Write=true
    Module.Index0Float2Write=true
    Module.Index0Float3Write=false
   )
   &FN_Mesh_Attributes(
    Create Dynamic Parameter Attribute=true
    Create Scale Attribute=true
    Index 0 Param 1:NiagaraFloat=&FloatFromCurve(
     FloatCurve=DataInterface() {propertyValues {Curve {keys [{arriveTangent=-2.1015, time=0.2, value=0.825}, <at 1 -0.15>]}}}
     CurveIndex=$Emitter.ReactivityAge
    )
    Index 0 Param 2:NiagaraFloat=&FloatFromCurve(
     FloatCurve=DataInterface() {propertyValues {Curve {keys [{arriveTangent=-0.3016, time=0.435, value=1}, <at 1 0.9>]}}}
     CurveIndex=$Emitter.ReactivityAge
    )
    Index 0 Param 3:NiagaraFloat=&FloatFromCurve(
     FloatCurve=DataInterface() {propertyValues {Curve {keys [{arriveTangent=3.1095, interpMode=RCIM_Cubic, leaveTangent=3.1095, tangentMode=RCTM_User, time=0.435, value=0}, <at 1 0.8>]}}}
     CurveIndex=$Emitter.ReactivityAge
    )
    Module Output=::NewEnumerator1
    Module.Index0Float0Write=true
    Module.Index0Float1Write=true
    Module.Index0Float2Write=true
    Module.Index0Float3Write=true
    Scale=&VectorFromFloat(Value=&FloatFromCurve(
      FloatCurve=DataInterface() {propertyValues {Curve {
         defaultValue=1
         keys [{arriveTangent=2.8105, interpMode=RCIM_Cubic, leaveTangent=2.8105, tangentMode=RCTM_User, time=0, value=0.5}, <at 1 1.5>]
        }}}
      CurveIndex=$Emitter.ReactivityAge
      Scale Curve=0.6
     ))
   )
   &FN_Mesh_Attributes(
    Color=&MakeLinearColorFromVectorAndFloat(
     "Vector (RGB)"=&Multiply_VectorByFloat(Vector=&MakeVectorFromLinearColorRGB(Color=$Particles.Color), Float=1)
     "Float (Alpha)"=&FloatFromCurve(FloatCurve=DataInterface() {propertyValues {Curve {keys [<at 0 1>,<at 0.435 0>]}}}, CurveIndex=$Emitter.ReactivityAge)
    )
    Create Color Attribute=true
    Create Dynamic Parameter Attribute=true
    Create Orientation Attribute=true
    Create Scale Attribute=true
    Index 0 Param 1:NiagaraFloat=&FloatFromCurve(
     FloatCurve=DataInterface() {propertyValues {Curve {keys [{arriveTangent=-2.8794, interpMode=RCIM_Cubic, leaveTangent=-2.8794, tangentMode=RCTM_User, time=0, value=0.5}, <at 0.435 -0.2917>]}}}
     CurveIndex=$Emitter.ReactivityAge
    )
    Index 0 Param 2:NiagaraFloat=&FloatFromCurve(
     FloatCurve=DataInterface() {propertyValues {Curve {keys [{arriveTangent=-0.3464, interpMode=RCIM_Cubic, leaveTangent=-0.3464, tangentMode=RCTM_User, time=0, value=1}, <at 0.435 0>]}}}
     CurveIndex=$Emitter.ReactivityAge
    )
    Index 0 Param 3:NiagaraFloat=&FloatFromCurve(
     FloatCurve=DataInterface() {propertyValues {Curve {
        defaultValue=1
        keys [
         {arriveTangent=-14.1456, interpMode=RCIM_Cubic, leaveTangent=-14.1456, tangentMode=RCTM_User, time=0, value=2}
         {tangentMode=RCTM_User, time=0.435, value=-0.5}
        ]
       }}}
     CurveIndex=$Emitter.ReactivityAge
    )
    Index 0 Param 4=1
    Module Output=::NewEnumerator1
    Module.Index0Float0Write=true
    Module.Index0Float1Write=true
    Module.Index0Float2Write=true
    Module.Index0Float3Write=true
    Orientation:Quat4f=&MakeQuaternion(
     Quaternion From=1
     Axis=<1 0 0>
     Angle=0
     Pitch=&RandomRangeFloat(Minimum=0, Maximum=360)
     From=?
     To=?
     Stabilization Mode:ENiagara_OrientVectorCount=::NewEnumerator0
     Side Direction:Vector3f=?
     Up Direction:Vector3f=?
     Angle Type=::NewEnumerator0
     Normalize:NiagaraBool=true
    )
    Scale=&VectorFromFloat(Value=0.45)
   ) {module=FN_Mesh_Attributes001}
   &FN_Sprite_Attributes(
    Camera Offset=10
    Color=&MakeLinearColorFromVectorAndFloat(
     "Vector (RGB)"=&Multiply_VectorByFloat(Vector=&MakeVectorFromLinearColorRGB(Color=$Particles.Color), Float=5)
     "Float (Alpha)"=1
    )
    Create Camera Offset Attribute=true
    Create Color Attribute=true
    Create SpriteSize Attribute=true
    Module Output:ENiagara_AttributeSamplingApplyOutput=::NewEnumerator1
    SpriteSize=&Vector2DFromFloat(Value=&FloatFromCurve(
      FloatCurve=DataInterface() {propertyValues {Curve {keys [
          <at 0 0 RCIM_Constant>
          {arriveTangent=8.535, interpMode=RCIM_Cubic, leaveTangent=8.535, tangentMode=RCTM_User, time=0.2, value=1}
          {interpMode=RCIM_Cubic, tangentMode=RCTM_User, time=0.325, value=1.5}
          <at 1 0>
         ]}}}
      CurveIndex=$Emitter.ReactivityAge
      Scale Curve=18
     ))
   )
  ]
  rendererClasses [&NiagaraMeshRendererProperties, &NiagaraSpriteRendererProperties]
  renderers [
   {rendererClass=&NiagaraMeshRendererProperties, rendererIndex=0}
   {rendererClass=&NiagaraMeshRendererProperties, rendererIndex=1}
   {rendererClass=&NiagaraSpriteRendererProperties, rendererIndex=2}
   {rendererClass=&NiagaraMeshRendererProperties, rendererIndex=3}
  ]
 }
}
summary {
 systemName=NS_Backpack_VigorKnackGolf_IdleReactivity
 emitters [
  {emitterName=Base, rendererClasses [&NiagaraSpriteRendererProperties, &NiagaraMeshRendererProperties]}
  {emitterName=ShockwaveBurst, rendererClasses [&NiagaraMeshRendererProperties, &NiagaraSpriteRendererProperties]}
 ]
 userVariables [
  User.bIsFrontEnd {type=NiagaraBool}
  User.bIsFrontEndPreview {type=NiagaraBool}
  User.DebugReactivity {type=NiagaraBool}
  User.KillCount {type=NiagaraInt32}
  User.ReactivityTime {type=NiagaraFloat, defaultValue=0, description=""}
 ]
}
